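; Building actor Blueprint for the LightCastle fireplace prop, which silences its fire sound when destroyed.
; On the death effects event it calls the parent handler, then checks that the looping fire audio component is valid and quickly fades it out with a linear curve.
; The construction script is empty.
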(event EventOnDeathPlayEffects
  (params (Damage :type Float%20%28single-precision%29) (DamageTags :type Gameplay%20Tag%20Container%20Structure%20%28by%20ref%29) (Momentum :type Vector :default "0, 0, 0") (HitInfo :type Hit%20Result%20Structure%20%28by%20ref%29) (InstigatedBy :type Fort%20Pawn%20Object%20Reference) (DamageCauser :type Actor%20Object%20Reference) (EffectContext :type Gameplay%20Effect%20Context%20Handle%20Structure))
  (bind _spot_lightcastle_fireplace_a_fire_loop_cue (Variables|Default|GetSpot_LightCastle_FirePlace_A_Fire_Loop_Cue))
  (|CallParentFunction Damage DamageTags Momentum HitInfo InstigatedBy DamageCauser EffectContext)
  (Utilities|IsValid _spot_lightcastle_fireplace_a_fire_loop_cue
    (:Is%20Valid
      (if (Utilities|IsValid _spot_lightcastle_fireplace_a_fire_loop_cue :!class /Script/Engine.KismetSystemLibrary)
        (Audio|Components|Audio|FadeOut _spot_lightcastle_fireplace_a_fire_loop_cue 0.2 0.0 "Linear" :!class /Script/Engine.AudioComponent)))))

(fn ConstructionScript)
